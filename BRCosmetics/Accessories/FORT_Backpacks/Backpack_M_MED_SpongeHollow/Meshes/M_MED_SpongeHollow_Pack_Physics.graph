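# Physics asset for the SpongeHollow backpack cosmetic, covering its dangling hammer, wrench and lid.
# Tiny placeholder capsules on pack, base_hammer_l and base_wrench_l serve as joint anchors, while larger capsules shape the three dyn_ bones.
# Hammer and wrench are constrained to their base bones with narrow swing and twist limits.
# The lid joins to the pack with all swing and twist motion free, so its angular limits do not restrict it.
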PhysicsAsset /BRCosmetics/Accessories/FORT_Backpacks/Backpack_M_MED_SpongeHollow/Meshes/M_MED_SpongeHollow_Pack_Physics
bodies {
 base_hammer_l [{shapeType=Capsule, length=1, radius=1, shapeName=base_hammer_l_capsule, rotation=<rot 0 -90 0>}]
 base_wrench_l [{shapeType=Capsule, length=1, radius=1, shapeName=base_wrench_l_capsule, rotation=<rot 0 -90 0>}]
 dyn_hammer_l [
  {shapeType=Capsule, length=15, radius=2, shapeName=dyn_hammer_l_capsule, center=<12 0 0>, rotation=<rot 90 0 0>}
  {
   shapeType=Capsule
   length=6.8335
   radius=2.2147
   shapeName=dyn_hammer_l_capsule
   center=<23.4331 1.9683 -1.1523>
   rotation=<rot -6.4983 0 -0.0001>
  }
 ]
 dyn_lid_c [
  {
   shapeType=Capsule
   length=25.9599
   radius=1.7
   shapeName=dyn_lid_c_capsule
   center=<8.9265 -1.1542 10.085>
   rotation=<rot -0.0352 -179.3405 -88.8094>
  }
  {shapeType=Capsule, length=25, radius=2, shapeName=dyn_lid_c_capsule, center=<8.5351 0 2.6512>, rotation=<rot 0 0 -90>}
 ]
 dyn_wrench_l [{shapeType=Capsule, length=16, radius=2, shapeName=dyn_wrench_l_capsule, center=<13.7387 0 0>, rotation=<rot 90 0 0>}]
 pack [{shapeType=Capsule, length=1, radius=1, shapeName=pack_capsule, rotation=<rot 0 90.1614 -89.8379>}]
}
constraints [
 {bone1Name=dyn_wrench_l, bone2Name=base_wrench_l, swing1LimitDegrees=7.5, swing2LimitDegrees=0.5, twistLimitDegrees=1.2}
 {bone1Name=dyn_hammer_l, bone2Name=base_hammer_l, swing1LimitDegrees=9, swing2LimitDegrees=7, twistLimitDegrees=1.3}
 {
  bone1Name=dyn_lid_c
  bone2Name=pack
  swing1LimitDegrees=45
  swing1Motion=Free
  swing2LimitDegrees=45
  swing2Motion=Free
  twistLimitDegrees=45
  twistMotion=Free
 }
]
