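# Animation Blueprint for the SpongeHollow backpack, adding physics-driven secondary motion to its tools, lid and handles.
# An offset input pose feeds separate front-end, skydiving and default chains, blended by bIsInFrontEnd and bIsSkyDiving.
# Anim Dynamics springs move individual bones; the default chain also uses a RigidBody simulation and nudges wrench and hammer bases.
# Each update swaps simulation-space settings and AlphaSkyDive by skydiving state to retune that simulation and screwdriver sway.
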AnimBlueprint /BRCosmetics/Accessories/FORT_Backpacks/Backpack_M_MED_SpongeHollow/Meshes/M_MED_SpongeHollow_Pack_AnimBP
target_skeleton=&/M_MED_SpongeHollow_Pack_Skeleton
graphs [
 AnimGraph {nodes [
   Anim Dynamics - Bone: dyn_wrench_l(Alpha=1, ExternalForce="0,0,0") {
    angularDampingOverride=1
    angularSpringConstant=15000
    bAngularSpring=true
    bOverrideAngularDamping=true
    bUseGravityOverride=true
    gravityScale=0.35
    simulationSpace=RootRelative
    boundBone {boneName=dyn_wrench_l}
    componentAppliedLinearAccClamp=<2000 2000 2000>
    componentLinearAccScale=<0.8 0.8 0.8>
    componentLinearVelScale=<1 1 1>
    gravityOverride {z=-980}
    physicsBodyDefinitions [{
      boundBone {boneName=dyn_wrench_l}
      boxExtents=<30 30 30>
      constraintSetup {angularLimitsMax=<0 5 4>, angularLimitsMin=<-5 -0.5 0>, angularTarget=<1 0 0>}
      localJointOffset=<3.7 0 0>
     }]
    simSpaceSettings {maxAngularAcceleration=100, maxAngularVelocity=10, simSpaceAngularAlpha=1}
   }
   Anim Dynamics - Bone: dyn_hammer_l(Alpha=1, ExternalForce="0,0,0") {
    angularDampingOverride=1
    angularSpringConstant=14000
    bAngularSpring=true
    bOverrideAngularDamping=true
    bUseGravityOverride=true
    gravityScale=0.35
    simulationSpace=RootRelative
    boundBone {boneName=dyn_hammer_l}
    componentAppliedLinearAccClamp=<4000 4000 4000>
    componentLinearAccScale=<0.8 0.8 0.8>
    componentLinearVelScale=<0.9 0.9 0.9>
    gravityOverride {z=-980}
    physicsBodyDefinitions [{
      boundBone {boneName=dyn_hammer_l}
      boxExtents=<33 33 33>
      constraintSetup {angularLimitsMax=<0 4 5>, angularLimitsMin=<0 -1 0>, angularTarget=<1 0 0>}
      localJointOffset=<4.5 0 0>
     }]
    simSpaceSettings {maxAngularAcceleration=100, maxAngularVelocity=10, simSpaceAngularAlpha=1}
   }
   "Transform (Modify) Bone - Bone: base_hammer_l"!(ComponentPose, bAlphaBoolEnabled:bool=True, Translation="0,0,0", Rotation="1,0.5,2", Scale="1,1,1") -> (Pose) {
    alphaInputType=Bool
    rotationMode=BMM_Additive
    rotationSpace=BCS_BoneSpace
    translationMode=BMM_Additive
    translationSpace=BCS_BoneSpace
    alphaBoolBlend {blendInTime=0.4, blendOption=QuadraticInOut, blendOutTime=0.25}
    boneToModify {boneName=base_hammer_l}
    rotation=<rot 1 0.5 2>
   }
   RigidBody() {
    overridePhysicsAsset=&/M_MED_SpongeHollow_Pack_Physics
    simSpaceSettings {
     dampingAlpha=0.5
     maxAngularAcceleration=0.25
     maxAngularVelocity=1.5
     maxLinearAcceleration=1500
     maxLinearVelocity=500
     velocityScaleZ=0.8
     worldAlpha=1
     externalLinearDragV=<0.4 0.4 0.4>
    }
   }
   n1 = Anim Dynamics - Bone: dyn_screwdriver_l(
    ComponentPose=Local To Component(LocalPose=Use cached pose 'Input pose'())
    Alpha=Get AlphaSkyDive() -> (AlphaSkyDive:float)
    ExternalForce="0,0,0"
   ) {
    angularDampingOverride=1
    angularSpringConstant=10000
    bAngularSpring=true
    bOverrideAngularDamping=true
    bUseGravityOverride=true
    gravityScale=0.35
    simulationSpace=RootRelative
    boundBone {boneName=dyn_screwdriver_l}
    componentAppliedLinearAccClamp=<2000 2000 2000>
    componentLinearAccScale=<0.5 0.5 0.5>
    componentLinearVelScale=<0.5 0.5 0.5>
    gravityOverride {z=-880}
    physicsBodyDefinitions [{
      boundBone {boneName=dyn_screwdriver_l}
      boxExtents=<25 25 25>
      constraintSetup {angularLimitsMax=<0 5 5>, angularLimitsMin=<-15 -5 -5>, angularTarget=<1 0 0>}
      localJointOffset=<3.4 0 0>
     }]
   }
   n2 = RigidBody(ComponentPose=@n1, Alpha=0) {
    overridePhysicsAsset=&M_MED_SpongeHollow_Pack_Physics
    simulationSpace=BaseBoneSpace
    componentAppliedLinearAccClamp=<3000 3000 3000>
    simSpaceSettings {worldAlpha=1}
   }
   n3 = Anim Dynamics - Bone: dyn_wrench_l(ComponentPose=@n2, Alpha=1, ExternalForce="0,0,0") {
    angularDampingOverride=1
    angularSpringConstant=12000
    bAngularSpring=true
    bOverrideAngularDamping=true
    bUseGravityOverride=true
    gravityScale=0.35
    simulationSpace=RootRelative
    boundBone {boneName=dyn_wrench_l}
    componentAppliedLinearAccClamp=<1000 1000 1000>
    componentLinearAccScale=<0.4 0.4 0.4>
    componentLinearVelScale=<0.5 0.5 0.5>
    gravityOverride {z=-980}
    physicsBodyDefinitions [{
      boundBone {boneName=dyn_wrench_l}
      boxExtents=<45 45 45>
      constraintSetup {angularLimitsMax=<0 4 3.2>, angularLimitsMin=<-3 0 0>, angularTarget=<1 0 0>}
      localJointOffset=<0.2 0 0>
     }]
    simSpaceSettings {maxAngularAcceleration=100, maxAngularVelocity=10, simSpaceAngularAlpha=0.7}
   }
   n4 = Anim Dynamics - Bone: dyn_hammer_l(ComponentPose=@n3, Alpha=1, ExternalForce="0,0,0") {
    angularDampingOverride=1
    angularSpringConstant=10000
    bAngularSpring=true
    bOverrideAngularDamping=true
    bUseGravityOverride=true
    gravityScale=0.35
    simulationSpace=RootRelative
    boundBone {boneName=dyn_hammer_l}
    componentAppliedLinearAccClamp=<1000 1000 1000>
    componentLinearAccScale=<0.7 0.7 0.7>
    componentLinearVelScale=<0.7 0.7 0.7>
    gravityOverride {z=-960}
    physicsBodyDefinitions [{
      boundBone {boneName=dyn_hammer_l}
      boxExtents=<38 38 38>
      constraintSetup {angularLimitsMax=<8 3 4.7>, angularLimitsMin=<-2 -0.7 0>, angularTarget=<1 0 0>}
      localJointOffset=<1 0 0>
     }]
    simSpaceSettings {maxAngularAcceleration=100, maxAngularVelocity=10, simSpaceAngularAlpha=0.7}
   }
   n5 = "Transform (Modify) Bone - Bone: base_hammer_l"(ComponentPose=@n4, bAlphaBoolEnabled=True, Translation="0,0,-0.8", Rotation="4.5,0.5,10", Scale="1,1,1") {
    alphaInputType=Bool
    rotationMode=BMM_Additive
    rotationSpace=BCS_BoneSpace
    translationMode=BMM_Additive
    translationSpace=BCS_BoneSpace
    alphaBoolBlend {blendInTime=0.4, blendOption=QuadraticInOut, blendOutTime=0.25}
    boneToModify {boneName=base_hammer_l}
    rotation=<rot 4.5 0.5 10>
    translation {z=-0.8}
   }
   n6 = Anim Dynamics - Bone: dyn_lid_c(ComponentPose=@n5, Alpha=0.65, ExternalForce="0,0,0") {
    angularDampingOverride=1
    angularSpringConstant=9000
    bAngularSpring=true
    bOverrideAngularDamping=true
    bUseGravityOverride=true
    gravityScale=0.35
    simulationSpace=RootRelative
    boundBone {boneName=dyn_lid_c}
    componentAppliedLinearAccClamp=<3000 3000 3000>
    componentLinearAccScale=<0.6 0.6 0.6>
    componentLinearVelScale=<0.7 0.7 0.7>
    gravityOverride {z=-980}
    physicsBodyDefinitions [{
      boundBone {boneName=dyn_lid_c}
      boxExtents=<29 29 29>
      constraintSetup {angularLimitsMax=<0 15 0>, angularLimitsMin=<0 -2 0>, angularTarget=<1 0 0>}
      localJointOffset=<3 0 0>
     }]
   }
   n7 = Anim Dynamics - Bone: dyn_handle_r(ComponentPose=@n6, Alpha=1, ExternalForce="0,0,0") {
    angularDampingOverride=1
    angularSpringConstant=1000
    bAngularSpring=true
    bOverrideAngularDamping=true
    bUseGravityOverride=true
    gravityScale=0.35
    simulationSpace=BoneRelative
    boundBone {boneName=dyn_handle_r}
    componentAppliedLinearAccClamp=<3000 3000 3000>
    componentLinearAccScale=<0.6 0.6 0.6>
    componentLinearVelScale=<0.6 0.6 0.6>
    gravityOverride {z=-980}
    physicsBodyDefinitions [{
      boundBone {boneName=dyn_handle_r}
      boxExtents=<20 20 20>
      constraintSetup {angularLimitsMax=<0 0 2>, angularLimitsMin=<0 0 -100>, angularTarget=<1 0 0>}
      localJointOffset=<5 0 0>
     }]
    simSpaceSettings {maxAngularAcceleration=30, maxAngularVelocity=3, simSpaceAngularAlpha=1}
   }
   n8 = Anim Dynamics - Bone: dyn_handle_l(ComponentPose=@n7, Alpha=1, ExternalForce="0,0,0") {
    angularDampingOverride=1
    angularSpringConstant=600
    bAngularSpring=true
    bOverrideAngularDamping=true
    bUseGravityOverride=true
    gravityScale=0.35
    simulationSpace=RootRelative
    boundBone {boneName=dyn_handle_l}
    componentAppliedLinearAccClamp=<3000 3000 3000>
    componentLinearAccScale=<0.8 0.8 0.8>
    componentLinearVelScale=<0.8 0.8 0.8>
    gravityOverride {z=-980}
    physicsBodyDefinitions [{
      boundBone {boneName=dyn_handle_l}
      boxExtents=<20 20 20>
      constraintSetup {angularLimitsMax=<0 0 100>, angularLimitsMin=<0 0 -2>, angularTarget=<1 0 0>}
      localJointOffset=<5 0 0>
     }]
    simSpaceSettings {maxAngularAcceleration=120, maxAngularVelocity=12}
   }
   n9 = "Transform (Modify) Bone - Bone: base_wrench_l"!(
    ComponentPose=Local To Component(LocalPose=Use cached pose 'Tools'())
    bAlphaBoolEnabled:bool=True
    Translation="1,0,-1"
    Rotation="-3.6,6,5"
    Scale="1,1,1"
   ) -> (Pose) {
    alphaInputType=Bool
    rotationMode=BMM_Additive
    rotationSpace=BCS_BoneSpace
    alphaBoolBlend {blendInTime=0.3, blendOutTime=0.2}
    boneToModify {boneName=base_wrench_l}
    rotation=<rot -3.6 6 5>
    translation {x=1, z=-1}
   }
   n10 = "Transform (Modify) Bone - Bone: base_hammer_l"(ComponentPose=@n9, bAlphaBoolEnabled=True, Translation="0,0.7,-1", Rotation="1.3,7.6,-4", Scale="1,1,1") {
    alphaInputType=Bool
    rotationMode=BMM_Additive
    rotationSpace=BCS_BoneSpace
    translationMode=BMM_Additive
    translationSpace=BCS_BoneSpace
    alphaBoolBlend {blendInTime=0.4, blendOption=QuadraticInOut, blendOutTime=0.25}
    boneToModify {boneName=base_hammer_l}
    rotation=<rot 1.3 7.6 -4>
    translation {y=0.7, z=-1}
   }
   n11 = "Transform (Modify) Bone - Bone: base_hammer_l"!(ComponentPose=@n10, Alpha:float=1, Translation="0,0,-1", Rotation="0,0,0", Scale="1,1,1") -> (Pose) {translationMode=BMM_Additive, translationSpace=BCS_BoneSpace, boneToModify {boneName=base_hammer_l}, translation {z=-1}}
   n12 = Anim Dynamics - Bone: dyn_lid_c(ComponentPose=@n11, Alpha=1, ExternalForce="0,0,0") {
    angularDampingOverride=1
    angularSpringConstant=9000
    bAngularSpring=true
    bOverrideAngularDamping=true
    bUseGravityOverride=true
    gravityScale=0.35
    simulationSpace=RootRelative
    boundBone {boneName=dyn_lid_c}
    componentAppliedLinearAccClamp=<3000 3000 3000>
    componentLinearAccScale=<0.6 0.6 0.6>
    componentLinearVelScale=<0.7 0.7 0.7>
    gravityOverride {z=-980}
    physicsBodyDefinitions [{
      boundBone {boneName=dyn_lid_c}
      boxExtents=<28 28 28>
      constraintSetup {angularLimitsMax=<0 21 0>, angularLimitsMin=<0 -11 0>, angularTarget=<1 0 0>}
      localJointOffset=<3.2 0 0>
     }]
    simSpaceSettings {maxAngularAcceleration=20, maxAngularVelocity=2, simSpaceAngularAlpha=0.2}
   }
   n13 = Anim Dynamics - Bone: dyn_handle_r(ComponentPose=@n12, Alpha=1, ExternalForce="0,0,0") {
    angularDampingOverride=1
    angularSpringConstant=1500
    bAngularSpring=true
    bOverrideAngularDamping=true
    bUseGravityOverride=true
    gravityScale=0.35
    simulationSpace=BoneRelative
    boundBone {boneName=dyn_handle_r}
    componentAppliedLinearAccClamp=<2500 2500 2500>
    componentLinearAccScale=<0.7 0.7 0.65>
    componentLinearVelScale=<0.7 0.7 0.65>
    gravityOverride {z=-980}
    physicsBodyDefinitions [{
      boundBone {boneName=dyn_handle_r}
      boxExtents=<20 20 20>
      constraintSetup {angularLimitsMax=<0 0 2>, angularLimitsMin=<0 0 -100>, angularTarget=<1 0 0>}
      localJointOffset=<4.7 0 0>
     }]
    simSpaceSettings {maxAngularAcceleration=100, maxAngularVelocity=10, simSpaceAngularAlpha=1}
   }
   n14 = Anim Dynamics - Bone: dyn_handle_l(ComponentPose=@n13, Alpha=1, ExternalForce="0,0,0") {
    angularDampingOverride=1
    angularSpringConstant=600
    bAngularSpring=true
    bOverrideAngularDamping=true
    bUseGravityOverride=true
    gravityScale=0.35
    simulationSpace=RootRelative
    boundBone {boneName=dyn_handle_l}
    componentAppliedLinearAccClamp=<3000 3000 3000>
    componentLinearAccScale=<0.7 0.7 0.65>
    componentLinearVelScale=<0.7 0.7 0.65>
    gravityOverride {z=-980}
    physicsBodyDefinitions [{
      boundBone {boneName=dyn_handle_l}
      boxExtents=<20 20 20>
      constraintSetup {angularLimitsMax=<0 0 100>, angularLimitsMin=<0 0 0>, angularTarget=<1 0 0>}
      localJointOffset=<5.5 0 0>
     }]
    simSpaceSettings {maxAngularAcceleration=100, maxAngularVelocity=10, simSpaceAngularAlpha=1}
   }
   n15 = Local To Component(LocalPose=Blend Poses by bool(
     BlendPose_0=Component To Local(ComponentPose=@n8)
     BlendPose_1=Component To Local(ComponentPose=@n14)
     BlendTime_0=0.25
     BlendTime_1=0.25
     bActiveValue=Get bIsSkyDiving()
    ))
   Output Pose(Result=Blend Poses by bool(
     BlendPose_0=Use cached pose 'FrontEnd_Dyn'()
     BlendPose_1=Component To Local(ComponentPose=@n15)
     BlendTime_0=0.25
     BlendTime_1=0.25
     bActiveValue=Get bIsInFrontEnd()
    )) {layerGroup=None}
   n16 = RigidBody!(
    ComponentPose=Local To Component(LocalPose=Use cached pose 'Input pose'())
    Alpha=Get AlphaSkyDive()
    ExternalForce
    SimSpaceSettings:Sim Space Settings Structure=Get Sim Space Settings() -> (Sim Space Settings:Sim Space Settings Structure)
   ) -> (Pose) {
    overridePhysicsAsset=&M_MED_SpongeHollow_Pack_Physics
    componentAppliedLinearAccClamp=<4000 4000 4000>
    simSpaceSettings {
     dampingAlpha=0
     maxAngularAcceleration=60
     maxAngularVelocity=15
     maxLinearAcceleration=400
     maxLinearVelocity=100
     velocityScaleZ=0.8
     worldAlpha=1
     externalLinearDragV=<0.5 0.5 0.5>
    }
   }
   n17 = Anim Dynamics - Bone: dyn_screwdriver_l(ComponentPose=@n16, Alpha=Get AlphaSkyDive(), ExternalForce="0,0,0") {
    angularDampingOverride=1
    angularSpringConstant=10000
    bAngularSpring=true
    bOverrideAngularDamping=true
    bUseGravityOverride=true
    gravityScale=0.35
    simulationSpace=RootRelative
    boundBone {boneName=dyn_screwdriver_l}
    componentAppliedLinearAccClamp=<2000 2000 2000>
    componentLinearAccScale=<0.5 0.5 0.5>
    componentLinearVelScale=<0.5 0.5 0.5>
    gravityOverride {z=-880}
    physicsBodyDefinitions [{
      boundBone {boneName=dyn_screwdriver_l}
      boxExtents=<25 25 25>
      constraintSetup {angularLimitsMax=<0 5 5>, angularLimitsMin=<-15 -5 -5>, angularTarget=<1 0 0>}
      localJointOffset=<3.4 0 0>
     }]
   }
   Save cached pose 'Tools'(Pose=Component To Local(ComponentPose=@n17)) {cacheName=Tools}
   n18 = Local To Component(LocalPose=Input Pose())
   Save cached pose 'Input pose'(Pose=Component To Local(ComponentPose="Transform (Modify) Bone - Bone: pack"(ComponentPose=@n18, Translation="0,1.2,0", Rotation="0,0,0") {translationMode=BMM_Additive, translationSpace=BCS_BoneSpace, boneToModify {boneName=pack}, translation {y=1.2}})) {cacheName=Input pose}
   n19 = RigidBody!(ComponentPose=Local To Component(LocalPose=Use cached pose 'Input pose'()), Alpha=0, ExternalForce) -> (Pose) {
    overridePhysicsAsset=&M_MED_SpongeHollow_Pack_Physics
    simulationSpace=BaseBoneSpace
    componentAppliedLinearAccClamp=<3000 3000 3000>
    simSpaceSettings {worldAlpha=1}
   }
   n20 = Anim Dynamics - Bone: dyn_wrench_l(ComponentPose=@n19, Alpha=0.73, ExternalForce="0,0,0") {
    angularDampingOverride=1
    angularSpringConstant=10000
    bAngularSpring=true
    bOverrideAngularDamping=true
    bUseGravityOverride=true
    gravityScale=0.35
    simulationSpace=RootRelative
    boundBone {boneName=dyn_wrench_l}
    componentAppliedLinearAccClamp=<1000 1000 1000>
    componentLinearAccScale=<0.4 0.4 0.4>
    componentLinearVelScale=<0.4 0.4 0.4>
    gravityOverride {z=-990}
    physicsBodyDefinitions [{
      boundBone {boneName=dyn_wrench_l}
      boxExtents=<39 39 39>
      constraintSetup {angularLimitsMax=<0 6 4.3>, angularLimitsMin=<0 -0.2 0.7>, angularTarget=<1 0 0>}
      localJointOffset=<2 0 0>
     }]
    simSpaceSettings {maxAngularAcceleration=15, maxAngularVelocity=1.5, simSpaceAngularAlpha=0.84}
   }
   n21 = Anim Dynamics - Bone: dyn_hammer_l(ComponentPose=@n20, Alpha=0.67, ExternalForce="0,0,0") {
    angularDampingOverride=1
    angularSpringConstant=11000
    bAngularSpring=true
    bOverrideAngularDamping=true
    bUseGravityOverride=true
    gravityScale=0.35
    simulationSpace=RootRelative
    boundBone {boneName=dyn_hammer_l}
    componentAppliedLinearAccClamp=<1000 1000 1000>
    componentLinearAccScale=<0.4 0.4 0.4>
    componentLinearVelScale=<0.4 0.4 0.4>
    gravityOverride {z=-980}
    physicsBodyDefinitions [{
      boundBone {boneName=dyn_hammer_l}
      boxExtents=<39 39 39>
      constraintSetup {angularLimitsMax=<0 3 9>, angularLimitsMin=<0 -2 1.3>, angularTarget=<1 0 0>}
      localJointOffset=<2.4 0 0>
     }]
    simSpaceSettings {maxAngularAcceleration=30, maxAngularVelocity=3, simSpaceAngularAlpha=0.84}
   }
   n22 = Anim Dynamics - Bone: dyn_screwdriver_l(ComponentPose=@n21, Alpha=0.65, ExternalForce="0,0,0") {
    angularDampingOverride=1
    angularSpringConstant=11000
    bAngularSpring=true
    bOverrideAngularDamping=true
    bUseGravityOverride=true
    gravityScale=0.35
    simulationSpace=RootRelative
    boundBone {boneName=dyn_screwdriver_l}
    componentAppliedLinearAccClamp=<1000 1000 1000>
    componentLinearAccScale=<0.4 0.4 0.4>
    componentLinearVelScale=<0.4 0.4 0.4>
    gravityOverride {z=-980}
    physicsBodyDefinitions [{
      boundBone {boneName=dyn_screwdriver_l}
      boxExtents=<33 33 33>
      constraintSetup {angularLimitsMax=<0 3 4>, angularLimitsMin=<-5 -3 -7>, angularTarget=<1 0 0>}
      localJointOffset=<2.5 0 0>
     }]
    simSpaceSettings {maxAngularAcceleration=15, maxAngularVelocity=1.5}
   }
   n23 = Anim Dynamics - Bone: dyn_lid_c(ComponentPose=@n22, Alpha=0.75, ExternalForce="0,0,0") {
    angularDampingOverride=1
    angularSpringConstant=9000
    bAngularSpring=true
    bOverrideAngularDamping=true
    bUseGravityOverride=true
    gravityScale=0.35
    simulationSpace=RootRelative
    boundBone {boneName=dyn_lid_c}
    componentAppliedLinearAccClamp=<3000 3000 3000>
    componentLinearAccScale=<0.6 0.6 0.6>
    componentLinearVelScale=<0.7 0.7 0.7>
    gravityOverride {z=-980}
    physicsBodyDefinitions [{
      boundBone {boneName=dyn_lid_c}
      boxExtents=<29 29 29>
      constraintSetup {angularLimitsMax=<0 24 0>, angularLimitsMin=<0 -7 0>, angularTarget=<1 0 0>}
      localJointOffset=<3 0 0>
     }]
   }
   n24 = Anim Dynamics - Bone: dyn_handle_r(ComponentPose=@n23, Alpha=0.9, ExternalForce="0,0,0") {
    angularDampingOverride=1
    angularSpringConstant=1600
    bAngularSpring=true
    bOverrideAngularDamping=true
    bUseGravityOverride=true
    gravityScale=0.35
    simulationSpace=BoneRelative
    boundBone {boneName=dyn_handle_r}
    componentAppliedLinearAccClamp=<2500 2500 2500>
    componentLinearAccScale=<0.9 0.9 0.9>
    componentLinearVelScale=<0.9 0.9 0.9>
    gravityOverride {z=-980}
    physicsBodyDefinitions [{
      boundBone {boneName=dyn_handle_r}
      boxExtents=<21 21 21>
      constraintSetup {angularLimitsMax=<0 0 1>, angularLimitsMin=<0 0 -90>, angularTarget=<1 0 0>}
      localJointOffset=<5.7 0 0>
     }]
    simSpaceSettings {maxAngularAcceleration=30, maxAngularVelocity=3}
   }
   n25 = Anim Dynamics - Bone: dyn_handle_l(ComponentPose=@n24, Alpha=0.9, ExternalForce="0,0,0") {
    angularDampingOverride=1
    angularSpringConstant=1600
    bAngularSpring=true
    bOverrideAngularDamping=true
    bUseGravityOverride=true
    gravityScale=0.35
    simulationSpace=RootRelative
    boundBone {boneName=dyn_handle_l}
    componentAppliedLinearAccClamp=<2200 2200 2200>
    componentLinearAccScale=<0.9 0.9 0.9>
    componentLinearVelScale=<0.9 0.9 0.9>
    gravityOverride {z=-980}
    physicsBodyDefinitions [{
      boundBone {boneName=dyn_handle_l}
      boxExtents=<22 22 22>
      constraintSetup {angularLimitsMax=<0 0 90>, angularLimitsMin=<0 0 -1>, angularTarget=<1 0 0>}
      localJointOffset=<5.2 0 0>
     }]
    simSpaceSettings {maxAngularAcceleration=120, maxAngularVelocity=12}
   }
   Save cached pose 'FrontEnd_Dyn'(Pose=Component To Local(ComponentPose=@n25)) {cacheName=FrontEnd_Dyn}
  ]}
 EventGraph {nodes [
   Event BlueprintUpdateAnimation()
   > n1 = Branch(Condition=Get bIsSkyDiving())
   > Set Sim Space Settings(Sim Space Settings:Sim Space Settings Structure=Get Sim Space Settings_0() -> (Sim Space Settings_0:Sim Space Settings Structure)) -> (Output_Get:Sim Space Settings Structure)
   > Set AlphaSkyDive(AlphaSkyDive:float=Get Alpha Sky Dive True() -> (Alpha Sky Dive True:float)) -> (Output_Get:float)
   Set Sim Space Settings(execute=@n1.else, Sim Space Settings=Get Sim Space Settings_1() -> (Sim Space Settings_1:Sim Space Settings Structure))
   > Set AlphaSkyDive(AlphaSkyDive=Get Alpha Sky Dive False() -> (Alpha Sky Dive False:float))
   Branch(Condition=Get bIsFalling())
   > Set Falling(Falling:bool=true) -> (Output_Get:bool)
   TryGetPawnOwner()
   Set Falling(Falling=false)
  ]}
]
variables [
 Alpha Sky Dive False {type=number, default=1}
 Alpha Sky Dive True {type=number, default=0.25}
 AlphaSkyDive {type=number, default=1}
 Falling {type=boolean}
 Sim Space Settings {
  type=SimSpaceSettings
  default {
   dampingAlpha=0
   maxAngularAcceleration=60
   maxAngularVelocity=15
   maxLinearAcceleration=400
   maxLinearVelocity=100
   velocityScaleZ=0.8
   worldAlpha=1
   externalLinearDragV=<0.5 0.5 0.5>
  }
 }
 Sim Space Settings_0 {
  type=SimSpaceSettings
  default {
   dampingAlpha=0.5
   maxAngularAcceleration=35
   maxAngularVelocity=70
   maxLinearAcceleration=7000
   maxLinearVelocity=2000
   velocityScaleZ=0.8
   worldAlpha=1
   externalLinearDragV=<0.6 0.6 0.6>
  }
 }
 Sim Space Settings_1 {
  type=SimSpaceSettings
  default {
   dampingAlpha=0
   maxAngularAcceleration=120
   maxAngularVelocity=40
   maxLinearAcceleration=400
   maxLinearVelocity=100
   velocityScaleZ=0.8
   worldAlpha=0.8
   externalLinearDragV=<0.5 0.5 0.5>
  }
 }
 Sim Space Settings_2 {
  type=SimSpaceSettings
  default {
   dampingAlpha=0.5
   maxAngularAcceleration=120
   maxAngularVelocity=24
   maxLinearAcceleration=2300
   maxLinearVelocity=100
   velocityScaleZ=0.9
   worldAlpha=1
   externalLinearDragV=<0.5 0.5 0.5>
  }
 }
]
